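# HeadMaterial is a Default Lit material in the ConcertSyncClient plugin, named for a head mesh.
# Its graph holds a single vector parameter, BodyColor, defaulting to opaque black, which material instances can override.
# No connection from that parameter to a material output appears in the graph, and there are no textures or other expressions.
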Material /ConcertSyncClient/HeadMaterial
graph {
 properties {pixelDepthOffsetMode=PDOM_Legacy, usedShadingModels=Default Lit}
 nodes [VectorParameter() {parameterName=BodyColor, defaultValue=<rgba 0 0 0 1>}]
}
