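# Masked unlit Concert Sync client material for distance-field text that faces the camera and scales with view depth.
# The offset math scales pivot-relative positions by CameraDepthFade, rotates them from view space around the component pivot and adds a depth-blended Z offset.
# A Roboto distance-field font sample drives a SphereMask outline over vertex color, with parameters for threshold, thickness and hardness.
# The graph also holds AlignMeshToTheCamera, ObjectScale and Color-tint nodes.
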Material /ConcertSyncClient/M_CameraFacing-scaling
graph {
 properties {blendMode=BLEND_Masked, pixelDepthOffsetMode=PDOM_Legacy, shadingModel=MSM_Unlit, usedShadingModels=Unlit}
 nodes [
  n1 = Multiply(
   A=Subtract(A=ObjectPositionWS(), B=WorldPosition().XYZ)
   B=LinearInterpolate(Alpha=MaterialFunctionCall() {materialFunction=&CameraDepthFade}.Result) {constA=-0.25, constB=-2}
  )
  n2 = MaterialFunctionCall!(
   Vector1=Transform(None=Constant3Vector() {constant=<rgba 0 0 1 0>}) {transformSourceType=TRANSFORMSOURCE_World, transformType=TRANSFORM_Local}
   Vector2=Transform(None=Constant3Vector() {constant=<rgba 1 0 0 0>}) {transformSourceType=TRANSFORMSOURCE_World, transformType=TRANSFORM_Local}
   Vector3=Transform(None=Constant3Vector() {constant=<rgba 0 -1 0 0>}) {transformSourceType=TRANSFORMSOURCE_World, transformType=TRANSFORM_Local}
  ) {materialFunction=/Engine/Functions/Engine_MaterialFunctions02/Math/MakeVectorsOrthogonal}
  n3 = Transform(None=MaterialFunctionCall(VectorToTransform=@n1, BasisX=@n2.Vector1, BasisY=@n2.Vector1, BasisZ=@n2.Vector1) {materialFunction=&Transform3x3Matrix}.Transformed Vector) {transformSourceType=TRANSFORMSOURCE_View}
  n4 = WorldPosition()
  n5 = MaterialFunctionCall(Float3=VertexNormalWS()) {materialFunction=&BreakOutFloat3Components}
  n6 = LinearInterpolate(
   A=ScalarParameter() {defaultValue=-10, parameterName=ZOffset_Min}
   B=ScalarParameter() {defaultValue=10, parameterName=ZOffset_Max}
   Alpha=MaterialFunctionCall() {materialFunction=&CameraDepthFade}.Result
  ) {constB=0}
  Add(
   A=Add(
    A=Subtract(A=@n3, B=@n4.XYZ)
    B=MaterialFunctionCall() {materialFunction=&ComponentPivotLocation}.Result
   )
   B=MaterialFunctionCall(X=@n5.R, Y=@n5.R, Z=Add(A=@n5.B, B=@n6)) {materialFunction=&MakeFloat3}.Result
  )
  n7 = FontSampleParameter() {font=/Engine/EngineFonts/RobotoDistanceField, parameterName=Font}
  LinearInterpolate(Alpha=@n7.R) {constA=-90, constB=90}
  n8 = VertexColor()
  LinearInterpolate(
   A=@n8
   B=Constant3Vector()
   Alpha=SphereMask(
    A=@n7.R
    B=ScalarParameter() {defaultValue=0.442, parameterName=OutlineMask}
    Radius=ScalarParameter() {defaultValue=0.09, parameterName=OutlineThickness}
    Hardness=ScalarParameter() {defaultValue=0.7, parameterName=OutlineHardness}
   )
  )
  MaterialFunctionCall!(
   WorldPosition=WorldPosition() {worldPositionShaderOffset=WPT_ExcludeAllShaderOffsets}.XYZ
   PivotLocation
   WS Normals
   Custom Object Basis 1=Constant3Vector() {constant=<rgba 0 0 1 0>}
   Custom Object Basis 2=Constant3Vector() {constant=<rgba 1 0 0 0>}
   Custom Object Basis 3=Constant3Vector() {constant=<rgba 0 -1 0 0>}
  ) {materialFunction=/Engine/Functions/Engine_MaterialFunctions02/WorldPositionOffset/AlignMeshToTheCamera}
  MaterialFunctionCall() {materialFunction=/Engine/Functions/Engine_MaterialFunctions02/WorldPositionOffset/ObjectScale}
  Multiply(A=VectorParameter() {parameterName=Color, defaultValue=<rgba 1 1 1 0>}.RGB, B=@n8)
  Multiply(
   A=@n1
   B=TransformPosition(None=WorldPosition() {worldPositionShaderOffset=WPT_ExcludeAllShaderOffsets}.XYZ) {transformSourceType=TRANSFORMPOSSOURCE_World}
  ) {desc=Apply Object Scale once again}
 ]
}
