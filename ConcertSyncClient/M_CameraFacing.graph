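# M_CameraFacing is an unlit, masked text material in the ConcertSyncClient plugin, built for camera-facing distance-field text.
# A Roboto distance-field font parameter feeds a parameterized SphereMask that blends vertex color toward black as an outline, and a -90 to 90 lerp.
# AlignMeshToTheCamera computes billboarding from world position and a custom basis.
# A separate branch tints vertex color with a Color parameter; the graph shows no links to material outputs.
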Material /ConcertSyncClient/M_CameraFacing
graph {
 properties {blendMode=BLEND_Masked, pixelDepthOffsetMode=PDOM_Legacy, shadingModel=MSM_Unlit, usedShadingModels=Unlit}
 nodes [
  n1 = FontSampleParameter() {font=/Engine/EngineFonts/RobotoDistanceField, parameterName=Font}
  LinearInterpolate(Alpha=@n1.R) {constA=-90, constB=90}
  n2 = VertexColor()
  LinearInterpolate(
   A=@n2
   B=Constant3Vector()
   Alpha=SphereMask(
    A=@n1.R
    B=ScalarParameter() {defaultValue=0.442, parameterName=OutlineMask}
    Radius=ScalarParameter() {defaultValue=0.09, parameterName=OutlineThickness}
    Hardness=ScalarParameter() {defaultValue=0.7, parameterName=OutlineHardness}
   )
  )
  MaterialFunctionCall!(
   WorldPosition=WorldPosition() {worldPositionShaderOffset=WPT_ExcludeAllShaderOffsets}.XYZ
   PivotLocation
   WS Normals
   Custom Object Basis 1=Constant3Vector() {constant=<rgba 0 0 1 0>}
   Custom Object Basis 2=Constant3Vector() {constant=<rgba 1 0 0 0>}
   Custom Object Basis 3=Constant3Vector() {constant=<rgba 0 -1 0 0>}
  ) {materialFunction=/Engine/Functions/Engine_MaterialFunctions02/WorldPositionOffset/AlignMeshToTheCamera}
  Multiply(A=VectorParameter() {parameterName=Color, defaultValue=<rgba 1 1 1 0>}.RGB, B=@n2)
 ]
}
