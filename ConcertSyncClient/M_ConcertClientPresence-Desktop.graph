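# Masked, default-lit material for desktop client presence in Concert Sync, judging by its name and path.
# A Fresnel on the pixel world normal drives a lerp toward the Color parameter.
# An inverted sphere mask between object and camera positions yields a proximity fade, tuned by FadeRadius and FadeSteps.
# Dithered opacity masking is enabled, allowing soft fades without translucent blending.
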Material /ConcertSyncClient/M_ConcertClientPresence-Desktop
graph {
 properties {blendMode=BLEND_Masked, ditherOpacityMask=true, pixelDepthOffsetMode=PDOM_Legacy, usedShadingModels=Default Lit}
 nodes [
  Constant3Vector() {constant=<rgba 0 0 0.001 0>}
  Constant() {r=0.6}
  LinearInterpolate(
   B=VectorParameter() {parameterName=Color, defaultValue=<rgba 0.0365 0.0365 0.0365 1>}.RGB
   Alpha=Fresnel(Normal=PixelNormalWS()) {baseReflectFraction=0.2, exponent=2}
  )
  OneMinus(None=SphereMask(
    A=ObjectPositionWS()
    B=CameraPositionWS()
    Radius=ScalarParameter() {defaultValue=240, parameterName=FadeRadius}
    Hardness=ScalarParameter() {defaultValue=0.1, parameterName=FadeSteps}
   ) {attenuationRadius=10, hardnessPercent=50})
 ]
}
