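# Default Lit material for client presence in the Concert sync client plugin, tinted through a Color vector parameter.
# A Fresnel term on the pixel world normal drives a lerp from the unconnected default A input toward Color, varying the tint with view angle.
# Two standalone constants, a near-black vector and a scalar, sit alongside the lerp; the file does not show which material outputs these nodes feed.
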Material /ConcertSyncClient/M_ConcertClientPresence
graph {
 properties {pixelDepthOffsetMode=PDOM_Legacy, usedShadingModels=Default Lit}
 nodes [
  Constant3Vector() {constant=<rgba 0 0 0.001 0>}
  Constant() {r=0.6}
  LinearInterpolate(
   B=VectorParameter() {parameterName=Color, defaultValue=<rgba 0.0365 0.0365 0.0365 1>}.RGB
   Alpha=Fresnel(Normal=PixelNormalWS()) {baseReflectFraction=0.2, exponent=2}
  )
 ]
}
