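# M_ConePresence is an unlit, two-sided translucent material in ConcertSyncClient, apparently for the cone marking another user's presence.
# A Fresnel term drives a rim glow tinted by HeadGlowColor and scaled by GlowAmount.
# The same Fresnel blends a faint base value toward full, multiplied by vertex-color red as a gradient mask.
# A sphere mask between object and camera positions fades that value out as the camera nears, tuned by FadeRadius and FadeSteps.
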Material /ConcertSyncClient/M_ConePresence
graph {
 properties {
  blendMode=BLEND_Translucent
  ditherOpacityMask=true
  pixelDepthOffsetMode=PDOM_Legacy
  shadingModel=MSM_Unlit
  twoSided=true
  usedShadingModels=Unlit
 }
 nodes [
  n1 = ScalarParameter() {defaultValue=0.125}
  n2 = Fresnel(Normal=PixelNormalWS()) {baseReflectFraction=0.2, exponent=2}
  n3 = ObjectPositionWS()
  n4 = CameraPositionWS()
  n5 = ScalarParameter() {defaultValue=220, parameterName=FadeRadius}
  n6 = ScalarParameter() {defaultValue=0.1, parameterName=FadeSteps}
  LinearInterpolate(
   B=Multiply(A=LinearInterpolate(A=@n1, Alpha=@n2) {constA=0.25}, B=VertexColor().R)
   Alpha=OneMinus(None=SphereMask(A=@n3, B=@n4, Radius=@n5, Hardness=@n6) {attenuationRadius=10, hardnessPercent=50})
  )
  Multiply(
   A=Multiply(
    A=VectorParameter() {parameterName=HeadGlowColor, defaultValue=<rgba 0.6227 0.98 0.6947 1>}.RGB
    B=ScalarParameter() {defaultValue=0.5, parameterName=GlowAmount}
   )
   B=@n2
  )
  VectorParameter() {parameterName=HeadBaseColor, defaultValue=<rgba 0.2159 0.2159 0.2159 0>}
 ]
}
