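# Masked, default-lit material for a head and body in the ConcertSyncClient plugin, with dithered opacity and spline mesh support.
# A Fresnel term on the pixel world normal scales a tintable glow color and amount for a rim highlight.
# The same Fresnel blends a scalar parameter toward a constant, then multiplies by vertex color red.
# Separate parameters supply a flat base color and a normal map, defaulting to a mannequin texture.
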Material /ConcertSyncClient/M_Head_Body
graph {
 properties {
  bUsedWithSplineMeshes=true
  blendMode=BLEND_Masked
  ditherOpacityMask=true
  pixelDepthOffsetMode=PDOM_Legacy
  refractionMethod=RM_IndexOfRefraction
  usedShadingModels=Default Lit
 }
 nodes [
  n1 = Fresnel(
   ExponentIn=ScalarParameter() {defaultValue=2, parameterName=Exponent}
   BaseReflectFractionIn=ScalarParameter() {defaultValue=0.2}
   Normal=PixelNormalWS()
  ) {baseReflectFraction=0.2, exponent=2}
  Multiply(
   A=Multiply(
    A=VectorParameter() {parameterName=HeadGlowColor_1, defaultValue=<rgba 0.49 0.98 0.8902 1>}.RGB
    B=ScalarParameter() {defaultValue=0.25, parameterName=GlowAmount_1}
   )
   B=@n1
  )
  Multiply(A=LinearInterpolate(A=ScalarParameter() {defaultValue=0.45, parameterName=Multiply}, Alpha=@n1) {constA=0.25}, B=VertexColor().R)
  TextureSampleParameter2D() {parameterName=NormalMap, samplerType=SAMPLERTYPE_Normal, texture=&/Mannequin__normals}
  VectorParameter() {parameterName=HeadBaseColor, defaultValue=<rgba 0.2159 0.2159 0.2159 0>}
 ]
}
