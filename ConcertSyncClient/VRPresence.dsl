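; VR presence Blueprint in ConcertSyncClient, with left and right hand meshes that represent a user by headset device type.
; The construction script checks whether the presence device type is OculusHMD.
; If it matches, the right hand mesh gets a fixed relative offset; otherwise the left hand mesh gets the same offset.
; BeginPlay, Tick and overlap events are empty.
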(event EventBeginPlay)

(event Collision|EventActorBeginOverlap
  (params (OtherActor :type Actor%20Object%20Reference)))

(event EventTick
  (params (DeltaSeconds :type Float%20%28single-precision%29)))

(fn ConstructionScript
  ;[1 1 1] Set Hands Relative Position
  (if (Utilities|Name|Equal(Name) (Variables|Rendering|GetPresenceDeviceType) "OculusHMD")
    (Transformation|SetRelativeLocation (Variables|Default|GetSM_Hand-Rgt-Mesh) "3.000000, 0,-1.500000" false false)
    (Transformation|SetRelativeLocation (Variables|Default|GetSM_Hand-Lft-Mesh) "3.000000, 0,-1.500000" false false)))
